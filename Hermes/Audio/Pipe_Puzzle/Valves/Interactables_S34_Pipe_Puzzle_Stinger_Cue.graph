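# Sound cue for the pipe puzzle stinger, filed with the valve interactable audio in Hermes.
# A single wave player feeds the S34 stinger wave straight to the output.
# It routes through the critical in-game UI sound class and sets a very large max distance.
# A concurrency override allows one instance at a time and blocks new plays while it is active.
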SoundCue /Hermes/Audio/Pipe_Puzzle/Valves/Interactables_S34_Pipe_Puzzle_Stinger_Cue
root=WavePlayer_1
nodes [{id=WavePlayer_1, properties {soundWaveAssetPtr=&/Interactables_S34_Pipe_Puzzle_Stinger}}]
properties {
 bOverrideConcurrency=true
 maxDistance=2097152
 soundClassObject=&Fort_UI_InGame_Critical
 volumeMultiplier=1
 concurrencyOverrides {maxCount=1, resolutionRule=PreventNew}
}
