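# Sound cue for the backward end sound of a valve in the S34 pipe puzzle, routed through the quests sound class.
# A single wave player feeds random, modulator and mixer nodes, with the modulator adding slight pitch variation and the cue boosting volume.
# It borrows attenuation from a supply drop ambient loop and overrides concurrency to one instance per owner, preventing new plays while one is active.
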SoundCue /Hermes/Audio/Pipe_Puzzle/Valves/Interactables_S34_Pipe_Puzzle_Valve_Backward_End_Cue
root=Mixer_0
nodes [
 {id=Mixer_0, children [Modulator_0], properties {inputVolume [1]}}
 {id=Modulator_0, children [Random_0], properties {pitchMax=1.02, pitchMin=0.99}}
 {id=Random_0, children [WavePlayer_2], properties {weights [1]}}
 {id=WavePlayer_2, properties {soundWaveAssetPtr=&/Interactables_S34_Pipe_Puzzle_Valve_Backward_End_01}}
]
properties {
 attenuationSettings=/Game/Athena/Sounds/SupplyDrops/Fort_HiddenStar_Ambient_Loop_Att
 bOverrideConcurrency=true
 maxDistance=2500
 soundClassObject=&Fort_Quests
 volumeMultiplier=1.5
 concurrencyOverrides {bLimitToOwner=true, maxCount=1, resolutionRule=PreventNew}
}
